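; Parent pickup container that lets an interacting player transform into a boss in Play As Villain.
; On spawn it applies pushback effects to targeted actors, starts music and pings, and saves the boss name server-side; traces set a timer-driven descent to ground.
; Server-side interaction locks it and activates the boss definition's intro ability, gated by transformation and inventory checks.
; RepNotifies drive movement, capsule state and cleanup, which destroys nearby player builds and the actor.
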(event Custom|UpdateBgaLocation
  (bind (_outhits _returnvalue) (Collision|MultiLineTraceForObjects (Transformation|GetActorLocation) (- (Transformation|GetActorLocation) "0.000000,0.000000,10000.000000") (Utilities|Array|MakeArray "ObjectTypeQuery1" "ObjectTypeQuery13")))
  (if _returnvalue
    (Variables|Default|SetBgaEndLocation (Utilities|Operators|Add (.Location (Collision|BreakHitResult (Utilities|Array|Get(acopy) _outhits))) (Math|Vector|MakeVector 0.0 0.0 (ScalableFloat|GetValueatLevel (Variables|Hotfix|GetSpawnOffset_Hf)))))
    (bind _returnvalue_2 (Utilities|Time|SetTimerbyEvent (delegate Custom|MoveBGA) (ScalableFloat|GetValueatLevel (Variables|Hotfix|GetMovePeriod_Hf)) true true))
    (Variables|Default|SetMoveBgaTHandle _returnvalue_2)))

(event Custom|MoveBGA
  (bind (_returnvalue_x _returnvalue_y _returnvalue_z) (Math|Vector|BreakVector (Transformation|GetActorLocation)))
  (bind (_bgaendlocation_x _bgaendlocation_y _bgaendlocation_z) (Math|Vector|BreakVector (Variables|GetBgaEndLocation_X)))
  (Variables|Default|SetBgaLocation (Utilities|Operators|Subtract (Transformation|GetActorLocation) (Math|Vector|MakeVector 0.0 0.0 (ScalableFloat|GetValueatLevel (Variables|Hotfix|GetMoveAmount_Hf)))))
  (if (>= _bgaendlocation_z _returnvalue_z)
    (Variables|Default|SetBgaLocation (Variables|Default|GetBgaEndLocation))
    (Utilities|Time|ClearandInvalidateTimerbyHandle (Variables|Default|GetMoveBgaTHandle))))

(event EventBeginPlay
  (bind _returnvalue (Game|GetInstigator))
  (bind _self self)
  (CallFunction|SetDestroyVolumeParams)
  (Variables|Default|SetbSetCapsuleOff true)
  (|CallParentFunction)
  (Utilities|FlowControl|Sequence
    (:then_0
      (Networking|SwitchHasAuthority
        (:Authority
          (Utilities|FlowControl|Sequence
            (:then_0
              ;[0.05718 0.1229 0.15 0.5] On Land Knockback Players and Vehicles
              (bind (_asynctaskproxy _targetinghandle) (Class|AsyncActionPerformTargeting|PerformTargetingAsyncAction 0 "/PlayAsVillain/Gameplay/Targeting/TP_PlayAsVillain_Container_Pushback.TP_PlayAsVillain_Container_Pushback" true)
                (:Targeted
                  (for _array_element (TargetingSystem|TargetingResults|GetTargetingResultsActors (GameInstanceSubsystems|GetTargetingSubsystem) _targetinghandle)
                    (Utilities|IsValid _array_element
                      (:Is%20Valid
                        (Utilities|IsValid (Ability|GetAbilitySystemComponent _array_element)
                          (:Is%20Valid
                            (bind _asfort_ability_system_component (Utilities|Casting|CastToFortAbilitySystemComponent (Ability|GetAbilitySystemComponent _self)))
                            (Fortnite|ApplyGameplayEffectsToActor _array_element (Variables|Default|GetGEsToApply) (Transformation|GetActorLocation) false 0 (Variables|Default|GetAbilitySystem) (GameplayEffects|MakeFortEffectContext _asfort_ability_system_component _self 0 _self 0))
                            (bind (_outhits _returnvalue_2) (Collision|MultiLineTraceForObjects (Transformation|GetActorLocation) (- (Transformation|GetActorLocation) "0.000000,0.000000,10000.000000") (Utilities|Array|MakeArray "ObjectTypeQuery1" "ObjectTypeQuery13")))
                            (if _returnvalue_2
                              (Variables|Default|SetBgaEndLocation (Utilities|Operators|Add (.Location (Collision|BreakHitResult (Utilities|Array|Get(acopy) _outhits))) (Math|Vector|MakeVector 0.0 0.0 (ScalableFloat|GetValueatLevel (Variables|Hotfix|GetSpawnOffset_Hf)))))
                              (bind _returnvalue_3 (Utilities|Time|SetTimerbyEvent (delegate Custom|MoveBGA) (ScalableFloat|GetValueatLevel (Variables|Hotfix|GetMovePeriod_Hf)) true true))
                              (Variables|Default|SetMoveBgaTHandle _returnvalue_3))))))))))
            (:then_1
              ; Grab boss name for interact.  Need to save it as text from the server as the client might not get a chance to get this information
              (Utilities|IsValid _returnvalue
                (:Is%20Valid
                  (bind _asfort_athena_aibot_controller (Utilities|Casting|CastToFortAthenaAIBotController (Fortnite|GetControllerfromActor _returnvalue)))
                  (Utilities|IsValid (.Character (Utilities|Struct|BreakFortAthenaLoadout (Variables|/Script/FortniteAI.FortAthenaAIBotController|GetCosmeticLoadoutBC _asfort_athena_aibot_controller)))
                    (:Is%20Valid
                      (Variables|InteractText|SetText_BossName (ItemDefinition|GetDisplayName (.Character (Utilities|Struct|BreakFortAthenaLoadout (Variables|/Script/FortniteAI.FortAthenaAIBotController|GetCosmeticLoadoutBC _asfort_athena_aibot_controller))))))))))))
        (:Remote
          (bind _returnvalue_4 (Audio|SpawnSoundatLocation "/PlayAsVillain/Assets/Audio/Music/MSS_PlayAsVillain_Music_Token_loop.MSS_PlayAsVillain_Music_Token_loop" (Transformation|GetActorLocation)))
          (bind _output_get (Variables|Default|SetMusicComponent _returnvalue_4))
          (Audio|AddEventbyComponent (GameInstanceSubsystems|GetMusicEventSubsystem) _output_get "(TagName=\"MusicEventSubsystem.Event.POI.Diegetic\")" "(GameplayTags=)" 0 ""))))
    (:then_1
      (Utilities|Time|SetTimerbyEvent (delegate Custom|ActivatePickup) 0.5)
      (SoundIndicator|StartSoundIndicatorLooping (Variables|Default|GetFortSoundIndicator) (Transformation|GetActorLocation self) 0 1.0 0))))

(event Custom|ActivatePickup
  (Variables|Default|SetbSetCapsuleOff false)
  ;[0.3843 0.151 1] UI Ping
  (HUDPings|MarkActor self) ; HUD Ping
)

(event Interaction|EventBlueprintOnInteract
  (params (InteractingPawn :type Fort%20Pawn%20Object%20Reference) (InteractionBeingAttempted :type EInteractionBeingAttempted%20Enum :default "FirstInteraction"))
  (bind _playerpawn (Default|GetPlayerPawnfromActor InteractingPawn))
  (bind _player_pawn (Variables|Default|GetPlayer%20Pawn))
  (bind _returnvalue (Ability|GetAbilitySystemComponent _player_pawn))
  (Networking|SwitchHasAuthority
    (:Authority
      (Variables|Default|GetPlayAsBossDef
        (:then
          (Utilities|IsValid _playerpawn
            (:Is%20Valid
              (Variables|Default|SetPlayer%20Pawn _playerpawn)
              (if (not (Math|Boolean|ORBoolean (Pawn|IsBotControlled (Variables|Default|GetPlayer%20Pawn)) (Health|IsDBNO (Variables|Default|GetPlayer%20Pawn)) (Variables|Default|GetbLocked)))
                (Variables|Default|SetbLocked true)
                (if (Vehicle|IsinVehicle (Variables|Default|GetPlayer%20Pawn))
                  (Utilities|IsValid _returnvalue
                    (:Is%20Valid
                      ;[1 1 1] Force exit from vehicle
                      (GameplayEffects|RemoveActiveEffectswithGrantedTags _returnvalue "(GameplayTags=((TagName=\"Gameplay.Action.Player.AthenaInVehicle\")))")
                      (bind _asfort_player_controller_athena (Utilities|Casting|CastToFortPlayerControllerAthena (Pawn|GetController (Variables|Default|GetPlayer%20Pawn)))
                        (:then
                          (Ability|Advanced|GrantandActivateAbilityonSelf (Ability|GetAbilitySystemComponent _asfort_player_controller_athena) (Variables|/PlayAsVillain/Gameplay/PlayAsBossDefinition.PlayAsBossDefinition_C|GetIntroAbility (Variables|Default|GetPlayAsBossDef)) _asfort_player_controller_athena)
                          (Variables|Default|SetbSetCapsuleOff true)
                          (Variables|Default|SetbReadyForCleanupSequence true)
                          ;[1 1 1] Failsafe Cleanup
                          (Utilities|Time|SetTimerbyEvent (delegate Custom|CleanupFailsafe) 1e+01))
                        (:CastFailed
                          _)))
                    (:Is%20Not%20Valid
                      _))
                  (Variables|Default|SetbLocked false)
                  (else
                    _))))))
        (:else)))))

(event Custom|CleanupFailsafe
  ;[1 1 1] Failsafe Cleanup
  (Variables|Default|SetbReadyForCleanupSequence true))

(event EventEndPlay
  (params (EndPlayReason :type EEndPlayReason%20Enum :default "Destroyed"))
  (SoundIndicator|StopSoundIndicatorLooping (Variables|Default|GetFortSoundIndicator))
  (bind _musiccomponent (Variables|Default|GetMusicComponent)
    (:then
      (Audio|Components|Audio|FadeOut _musiccomponent 0.3 0.0 "Linear" :!class /Script/Engine.AudioComponent)
      (Audio|RemoveEvent (GameInstanceSubsystems|GetMusicEventSubsystem) _musiccomponent)
      (Networking|SwitchHasAuthority
        (:Authority
          (Utilities|Time|ClearandInvalidateTimerbyHandle (Variables|Default|GetMoveBgaTHandle)))))
    (:else)))

(event Custom|Cleanup
  (Utilities|FlowControl|Sequence
    (:then_0
      (CallFunction|DestroyPlayerBuilds))
    (:then_1
      (HUDPings|UnmarkActor)
      (GameplayCue|ExecuteGameplayCue self (Variables|Default|GetGameplay%20Cue%20Tag))
      (Utilities|FlowControl|Delay 0.3)
      (Development|SetHiddeninGame (Variables|Default|GetScene) true true)
      (Utilities|FlowControl|Delay 1.0)
      (Actor|DestroyActor))))

(event InteractInterface|EventOnFillContextInfo
  (params (InteractType :type Interaction%20Type%20Structure%20%28by%20ref%29) (ContextInfo :type Fort%20Interact%20Context%20Info%20Object%20Reference))
  ;[1 1 1] Adjust interaction text - backpack full etc...
  (if (Math|Boolean|ORBoolean (Utilities|Text|EqualExactly(Text) (Variables|InteractText|GetAlreadyTransformedFailedInteractText) (Variables|/Script/FortniteGame.FortInteractContextInfo|GetContextText ContextInfo)) (Utilities|Text|EqualExactly(Text) (Variables|/Script/FortniteGame.FortInteractContextInfo|GetContextText ContextInfo) (Variables|InteractText|GetSquadAlreadyTransformedFailedText)) (Utilities|Text|EqualExactly(Text) (Variables|/Script/FortniteGame.FortInteractContextInfo|GetContextText ContextInfo) (Variables|InteractText|GetInventoryFullFailedInteractText)))
    (Variables|/Script/FortniteGame.FortInteractContextInfo|SetContextText (CallFunction|GetTransformInteractText) ContextInfo)
    (bind (_caninteract _isinteractingpawntransformed _issquadtransformed) (Utilities|HandleTransformInteract (Game|Player|GetPlayerPawn (Fortnite|GetFortPlayerControllerfromActor (Variables|/Script/FortniteGame.FortInteractContextInfo|GetInteractingActor ContextInfo)))))
    (Variables|/Script/FortniteGame.FortInteractContextInfo|SetErrorText (Utilities|String|SelectText (Variables|InteractText|GetAlreadyTransformedFailedInteractText) (Utilities|String|SelectText (Variables|InteractText|GetSquadAlreadyTransformedFailedText) (Variables|InteractText|GetInventoryFullFailedInteractText) _issquadtransformed) _isinteractingpawntransformed) ContextInfo)
    (Variables|/Script/FortniteGame.FortInteractContextInfo|SetErrorIconTag "(TagName=\"HUD.Athena.ErrorIcon.Interaction.BackpackFull\")" ContextInfo)))

(event OnSoundLibraryPlayEvent(SoundLibrary)
  (params (Results :type Sound%20Library%20Play%20Result%20Structure%20%28by%20ref%29) (Context :type Sound%20Library%20Context%20Object%20Reference))
  (if (GameplayTags|MatchesTag (Utilities|Struct|BreakSoundLibraryPlayResult Results) "(TagName=\"SoundLibrary.Item.Interact.Long\")" true)
    (Audio|Modulation|ActivateControlBusMix "/PlayAsVillain/Assets/Audio/SFX/CBM_PlayAsVillain_ItemInteraction.CBM_PlayAsVillain_ItemInteraction")))

(event OnLinkedActorDestroyed(FortLinkToActor)
  (params (DamageCauser :type Actor%20Object%20Reference))
  (CallFunction|UpdateBgaLocation))

(fn BlueprintCanInteract
  (params (InteractingPawn :type Fort%20Pawn%20Object%20Reference) (InteractionBeingAttempted :type EInteractionBeingAttempted%20Enum :default "FirstInteraction") (InteractionType :type TInteractionType%20Enum :default "IT_NoInteraction"))
  (bind _returnvalue (Fortnite|GetFortPlayerControllerfromActor InteractingPawn))
  (if (|CallParentFunction InteractingPawn InteractionBeingAttempted InteractionType)
    (Utilities|IsValid _returnvalue
      (:Is%20Valid
        (bind _caninteract (Utilities|HandleTransformInteract (Variables|GetInteractingPawn)))
        (if _caninteract
          (bind _cangrantloadout (CallFunction|CanGrantLoadout _returnvalue))
          (return _cangrantloadout)
          (else
            (return false))))
      (:Is%20Not%20Valid
        (return false)))
    (else
      (return false))))

(fn BlueprintGetFailedInteractionString
  (params (InteractingPawn :type Fort%20Pawn%20Object%20Reference) (InteractionBeingAttempted :type EInteractionBeingAttempted%20Enum :default "FirstInteraction"))
  (bind (_caninteract _isinteractingpawntransformed _issquadtransformed) (Utilities|HandleTransformInteract InteractingPawn))
  (return (select _isinteractingpawntransformed (Utilities|String|SelectText (Variables|InteractText|GetSquadAlreadyTransformedFailedText) (Variables|InteractText|GetInventoryFullFailedInteractText) _issquadtransformed) (Variables|InteractText|GetAlreadyTransformedFailedInteractText))))

(fn BlueprintGetFailedSubInteractionString
  (params (InteractingPawn :type Fort%20Pawn%20Object%20Reference) (InteractionBeingAttempted :type EInteractionBeingAttempted%20Enum :default "FirstInteraction"))
  (return ""))

(fn BlueprintGetInteractionString
  (params (InteractingPawn :type Fort%20Pawn%20Object%20Reference) (InteractionBeingAttempted :type EInteractionBeingAttempted%20Enum :default "FirstInteraction"))
  (|CallParentFunction InteractingPawn InteractionBeingAttempted)
  (return (CallFunction|GetTransformInteractText)))

(fn BlueprintGetSubInteractionString
  (params (InteractingPawn :type Fort%20Pawn%20Object%20Reference) (InteractionBeingAttempted :type EInteractionBeingAttempted%20Enum :default "FirstInteraction"))
  (bind _returnvalue (CallFunction|GetSubInteractionStringInternal InteractingPawn))
  (return _returnvalue))

(fn CanGrantLoadout
  (params (TargetPlayerController :type Fort%20Player%20Controller%20Object%20Reference))
  (Utilities|IsValid TargetPlayerController
    (:Is%20Valid
      (bind _playasbossdef (Variables|Default|GetPlayAsBossDef)
        (:then
          (for _array_element (Variables|/PlayAsVillain/Gameplay/PlayAsBossDefinition.PlayAsBossDefinition_C|GetLoadoutWhileTransformed _playasbossdef)
            (if (> (Inventory|CountInventoryOverflowfromAddingItem (Variables|GetTargetPlayerController) (Loot|CreateItemEntry _array_element) false false) 0)
              (return false)))
          (return true))
        (:else
          _)))
    (:Is%20Not%20Valid
      _))
  (return false))

(fn DestroyPlayerBuilds
  (bind (_returnvalue_x _returnvalue_y _returnvalue_z) (Math|Vector|BreakVector (Transformation|GetActorLocation self)))
  (if (> (ScalableFloat|GetValueatLevel (Variables|Hotfix|GetDestroyVolumeEnabled_Hf)) 0)
    (Networking|SwitchHasAuthority
      (:Authority
        (bind _outhits (Collision|MultiSphereTraceForObjects :Start (Math|Vector|MakeVector _returnvalue_x _returnvalue_y _returnvalue_z) :End (Math|Vector|MakeVector _returnvalue_x _returnvalue_y (+ _returnvalue_z (Variables|/Script/Engine.CapsuleComponent|GetCapsuleHalfHeight (Variables|Default|GetDestroyBuilds)))) :Radius (Variables|/Script/Engine.CapsuleComponent|GetCapsuleRadius (Variables|Default|GetDestroyBuilds)) :ObjectTypes (Utilities|Array|MakeArray "ObjectTypeQuery6" "ObjectTypeQuery2" "ObjectTypeQuery1" "ObjectTypeQuery3" "ObjectTypeQuery4") :bTraceComplex false :DrawDebugType "None" :bIgnoreSelf true :TraceColor "(R=1.000000,G=0.000000,B=0.000000,A=1.000000)" :TraceHitColor "(R=0.000000,G=1.000000,B=0.000000,A=1.000000)" :DrawTime 0.2))
        (for _array_element _outhits
          (bind _returnvalue (GameplayTags|HasAnyMatchingGameplayTags(Message) (.HitActor (Collision|BreakHitResult _array_element)) (GameplayTags|MakeLiteralGameplayTagContainer "(GameplayTags=((TagName=\"Building.Type.Player\")))")))
          (if _returnvalue
            (Utilities|IsValid (.HitActor (Collision|BreakHitResult _array_element))
              (:Is%20Valid
                (Actor|DestroyActor (.HitActor (Collision|BreakHitResult _array_element)))))))
        (return)))
    (else
      (return))))

(fn GetPlayerPawnFromActor
  (params (Actor :type Actor%20Object%20Reference))
  (bind _asfort_player_pawn (Utilities|Casting|CastToFortPlayerPawn Actor)
    (:then
      (return _asfort_player_pawn))
    (:CastFailed
      (bind _asfort_athena_vehicle (Utilities|Casting|CastToFortAthenaVehicle Actor)
        (:then
          (return (Vehicle|GetDriver _asfort_athena_vehicle)))
        (:CastFailed
          (return 0))))))

(fn GetSubInteractionString_Internal
  (params (InteractingPawn :type Fort%20Pawn%20Object%20Reference))
  (bind (_caninteract _isinteractingpawntransformed) (Utilities|HandleTransformInteract InteractingPawn))
  (return (Utilities|String|SelectText (Variables|InteractText|GetText_SelfTransformedSubDescription) (Variables|InteractText|GetText_DefaultSubDescription) _isinteractingpawntransformed)))

(fn GetTransformInteractText
  (return (Utilities|Text|FormatText (Variables|InteractText|GetText_Interaction) (Variables|InteractText|GetText_BossName))))

(fn MarkActor
  (params (ActorToMark :type Actor%20Object%20Reference))
  (bind _outfortplayerpawns (Fortnite|GetFortPlayerPawnsInRange (Transformation|GetActorLocation) (ScalableFloat|GetValueatLevel (Variables|Hotfix|GetMarkerRadius_Hf))))
  (for _array_element _outfortplayerpawns
    (Utilities|IsValid (Fortnite|GetFortPlayerControllerfromActor _array_element)
      (:Is%20Valid
        (Utilities|IsValid (Class|FortPlayerController|GetMarkerComponent (Fortnite|GetFortPlayerControllerfromActor _array_element))
          (:Is%20Valid
            (Utilities|IsValid ActorToMark
              (:Is%20Valid
                (bind (_playerbankid _bankinstanceid) (Fortnite|RegisterSoundIndicatorPlacementContext (Fortnite|GetFortPlayerControllerfromActor _array_element) ActorToMark (Transformation|GetActorLocation) false))
                (Variables|SetCrystalMarkerID_PlayerID (Utilities|Struct|MakeMarkerID _playerbankid _bankinstanceid))))))))))

(fn OnRep_BgaLocation
  (Transformation|SetActorLocation (Variables|Default|GetBgaLocation)))

(fn OnRep_bLocked
  (Variables|Interaction|SetbAllowInteract false)
  (Networking|CanExecuteCosmeticEvents
    (:True
      (Variables|Default|GetbLocked
        (:then
          (bind _musiccomponent (Variables|Default|GetMusicComponent)
            (:then
              (Audio|Components|Audio|FadeOut _musiccomponent 0.3 0.0 "Linear" :!class /Script/Engine.AudioComponent)
              (Audio|RemoveEvent (GameInstanceSubsystems|GetMusicEventSubsystem) _musiccomponent))
            (:else)))
        (:else)))))

(fn OnRep_bReadyForCleanupSequence
  (if (Variables|Default|GetbReadyForCleanupSequence)
    (Audio|Components|Audio|FadeOut (Variables|Default|GetAmbience) 0.5 0.0 "Logarithmic" :!class /Script/Engine.AudioComponent)
    (bind _musiccomponent (Variables|Default|GetMusicComponent)
      (:then
        (Audio|Components|Audio|FadeOut _musiccomponent 0.3 0.0 "Linear" :!class /Script/Engine.AudioComponent)
        (Audio|RemoveEvent (GameInstanceSubsystems|GetMusicEventSubsystem) _musiccomponent))
      (:else
        _))
    (CallFunction|Cleanup)))

(fn OnRep_bSetCapsuleOff
  (if (Variables|Default|GetbSetCapsuleOff)
    (Collision|SetCollisionEnabled (Variables|Default|GetOverlapCapsule) "NoCollision")
    (Development|SetHiddeninGame (Variables|Default|GetStaticMesh) true true)
    (Networking|FlushNetDormancy)
    (else
      (Collision|SetCollisionEnabled (Variables|Default|GetOverlapCapsule) "QueryAndPhysics")
      (Development|SetHiddeninGame (Variables|Default|GetStaticMesh) false false)
      (Networking|FlushNetDormancy))))

(fn SetDestroyVolumeParams
  (bind _destroybuilds (Variables|Default|GetDestroyBuilds))
  (Components|Capsule|SetCapsuleHalfHeight _destroybuilds (ScalableFloat|GetValueatLevel (Variables|Hotfix|GetDestroyVolumeHalfHeight_Hf)) true)
  (Components|Capsule|SetCapsuleRadius _destroybuilds (ScalableFloat|GetValueatLevel (Variables|Hotfix|GetDestroyVolumeRadius_Hf)) true)
  (return))

(fn UnmarkActor
  (bind _slatebrush (Utilities|Struct|MakeSlateBrush "(SpecifiedColor=(R=1.000000,G=1.000000,B=1.000000,A=1.000000),ColorUseRule=UseColor_Specified)" "Image" "NoTile" "NoMirror" (struct :X 0.0 :Y 0.0) "(Left=0.000000,Top=0.000000,Right=0.000000,Bottom=0.000000)" 0 "(CornerRadii=(X=0.000000,Y=0.000000,Z=0.000000,W=0.000000),Color=(SpecifiedColor=(R=0.000000,G=0.000000,B=0.000000,A=0.000000),ColorUseRule=UseColor_Specified),Width=0.000000,RoundingType=HalfHeightRadius,bUseBrushTransparency=False)"))
  (bind _athenaspecialactor (Variables|Default|GetAthenaSpecialActor))
  (for _array_element (Fortnite|GetAllFortPlayerControllers true true)
    (Utilities|IsValid _array_element
      (:Is%20Valid
        (Utilities|IsValid (Class|FortPlayerController|GetMarkerComponent _array_element)
          (:Is%20Valid
            (MarkerComponent|UnmarkActoronClient (Class|FortPlayerController|GetMarkerComponent _array_element) (Variables|Default|GetCrystalMarkerID)))))))
  (Variables|/Script/FortniteGame.AthenaSpecialActorComponent|SetSpecialActorCompassIconBrush _slatebrush _athenaspecialactor) ; Remove compass icon
  (Variables|/Script/FortniteGame.AthenaSpecialActorComponent|SetSpecialActorMinimapIconBrush _slatebrush _athenaspecialactor) ; Remove minimap icon
)

(fn ConstructionScript
  (|CallParentFunction))
