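; Parent component that spawns a Play As Villain transformation container at its owner's location when the AI boss is eliminated.
; On BeginPlay, the server registers the boss definition's transformed loadout and transformation item with the game state as custom granted items.
; On elimination, a scalable-float toggle gates the server-side spawn, which uses random yaw and receives the boss definition.
; An optional custom boss name is applied to the container after a short delay.
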(event EventBeginPlay
  (CallFunction|PreloadItemDefs))

(event EventPlayAsVillainAIBossEliminated
  (Utilities|SwitchHasAuthority
    (:Authority
      (Variables|Default|GetPlayAsBossDef
        (:then
          (if (ScalableFloat|GetValueasBool (Variables|Default|GetTransformBgaEnabled_Hf))
            (bind _returnvalue (Game|SpawnActorfromClass "/PlayAsVillain/Gameplay/BGA_PlayAsVillain_Transformation_Container_Parent.BGA_PlayAsVillain_Transformation_Container_Parent_C" (Math|Transform|MakeTransform (Transformation|GetActorLocation (Components|GetOwner)) (Math|Rotator|MakeRotator 0.0 0.0 (Math|Random|RandomFloatinRange 0.0 3.6e+02))) "Undefined" "MultiplyWithRoot" 0 (Variables|Default|GetPlayAsBossDef) 0 (Utilities|Casting|CastToPawn (Components|GetOwner))))
            (if (Variables|Default|GetUseCustomBossName)
              (Utilities|IsValid _returnvalue
                (:Is%20Valid
                  (Utilities|FlowControl|Delay 0.2)
                  (Variables|/PlayAsVillain/Gameplay/BGA_PlayAsVillain_Transformation_Container_Parent.BGA_PlayAsVillain_Transformation_Container_Parent_C|SetText_BossName (Variables|Default|GetBossName) _returnvalue))))))
        (:else)))))

(fn PreloadItemDefs
  (bind _returnvalue (Fortnite|Athena|GameState|GetGameStateAthena))
  (Utilities|SwitchHasAuthority
    (:Authority
      (bind _playasbossdef (Variables|Default|GetPlayAsBossDef)
        (:then
          (bind _output_get (LocalVariables|SetGrantedItems (Variables|/PlayAsVillain/Gameplay/PlayAsBossDefinition.PlayAsBossDefinition_C|GetLoadoutWhileTransformed _playasbossdef)))
          (Utilities|Array|Add _output_get (Variables|/PlayAsVillain/Gameplay/PlayAsBossDefinition.PlayAsBossDefinition_C|GetTransformationItemDef _playasbossdef))
          (Utilities|IsValid _returnvalue
            (:Is%20Valid
              (Class|FortGameStateAthena|RegisterCustomGrantedItems _returnvalue (LocalVariables|GetGrantedItems) 1.0))))
        (:else)))))
