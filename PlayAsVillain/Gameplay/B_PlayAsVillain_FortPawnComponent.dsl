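; Pawn component for Play As Villain that shows a boss-style status bar over transformed players and handles their transformation effects.
; Non-cosmetic instances write health and overshield into replicated structs, while cosmetic ones build the widget and update it through OnReps.
; Timers drive widget offset and visibility by range, line of sight and cloaking.
; It also handles scale overrides, ability-use warnings, sound indicators, footstep range and an optional minimap icon.
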(event Custom|EventAbilityUsed
  (params (IsAbilityUsed :type Boolean))
  (bind _pavabilityduration_hf (Variables|Hotfix|GetPAVAbilityDuration_Hf))
  (bind _returnvalue (ScalableFloat|GetValueatLevel _pavabilityduration_hf))
  (Variables|Default|SetbAbilityUsed IsAbilityUsed)
  (Utilities|FlowControl|Sequence
    (:then_0
      (Utilities|FlowControl|RetriggerableDelay _returnvalue)
      (Variables|Default|SetbAbilityUsed false))
    (:then_1
      (Utilities|FlowControl|RetriggerableDelay (* _returnvalue (ScalableFloat|GetValueatLevel (Variables|Hotfix|GetAbilityDurationWarningPercent_Hf))))
      (GameplayCue|ExecuteGameplayCueOnActor(Burst) (Components|GetOwner) (Variables|GameplayCues|GetAbilityEndWarning_GCN) (Ability|GameplayCue|MakeGameplayCueParameters 0.0)))))

(event Custom|SetScaleOverride
  (params (NewScaleOverride :type Float%20%28double-precision%29))
  (Networking|SwitchHasAuthority
    (:Authority
      (Variables|Default|SetScaleOverride NewScaleOverride))))

(event EventBeginPlay
  (Variables|Default|SetHasBegunPlay true)
  (CallFunction|HandleBossStatusWidgetEnabledChanged)
  (CallFunction|ListenforCloakedActivation))

(event EventEndPlay
  (params (EndPlayReason :type EEndPlayReason%20Enum :default "Destroyed"))
  (Variables|Default|SetHasBegunPlay false)
  (CallFunction|HandleAudioFootstepOverride 3584.0)
  (AsyncAction|Cancel (Variables|Cloaking|GetCloakActiveAction)))

(event Custom|ClearScaleOverride
  (Networking|SwitchHasAuthority
    (:Authority
      (Variables|Default|SetScaleOverride 0.0))))

(event Custom|SetBossStatusWidgetEnabled
  (params (NewIsEnabled :type Boolean) (TransformationTag :type Gameplay%20Tag%20Structure))
  (Variables|Default|SetbBossStatusWidgetEnabled NewIsEnabled)
  (Variables|Default|SetTransformation%20Tag TransformationTag))

(event Custom|StartUpdatingStatusWidget
  ;[1 1 1] Health Bar
  (bind _statuswidgetcomponent (Variables|FortPawnComponent_PlayAsVillain|GetStatusWidgetComponent)
    (:then
      (bind _asnpcstatus_widget_aiboss (Utilities|Casting|CastToNPCStatusWidget_AIBoss (UserInterface|GetUserWidgetObject _statuswidgetcomponent)))
      (Utilities|FlowControl|Sequence
        (:then_0
          (Components|Tick|SetComponentTickEnabled (Variables|FortPawnComponent_PlayAsVillain|GetStatusWidgetComponent) (Widget|IsVisible _asnpcstatus_widget_aiboss)))
        (:then_1
          ;;[1 1 1] DBNO Bind
          (bind _asfort_pawn (Utilities|Casting|CastToFortPawn (Components|GetOwner)))
          (CallFunction|HandlePawnDBNOStateChange _asfort_pawn (Health|IsDBNO _asfort_pawn))
          (EventDispatchers|BindEventtoOnDBNOStateChanged _asfort_pawn (EventDispatchers|CreateEvent :!event HandlePawnDBNOStateChange)))
        (:then_2
          (Default|BindEventtoVisibilityChanged _asnpcstatus_widget_aiboss (delegate Custom|WidgetVisibilityChanged)))))
    (:else)))

(event Custom|WidgetVisibilityChanged
  (params (bVisible :type Boolean))
  ;[1 1 1] Health Bar
  (Utilities|FlowControl|Sequence
    (:then_0
      (bind _statuswidgetcomponent (Variables|FortPawnComponent_PlayAsVillain|GetStatusWidgetComponent)
        (:then
          (Components|Tick|SetComponentTickEnabled _statuswidgetcomponent bVisible))
        (:else)))
    (:then_1
      (if bVisible
        (bind _returnvalue (Utilities|Time|SetTimerbyEvent (EventDispatchers|CreateEvent :!event UpdateWidgetOffset) 0.1 true true))
        (Variables|Default|SetStatusWidgetOffsetTimerHandle _returnvalue)
        (else
          (Utilities|Time|ClearandInvalidateTimerbyHandle (Variables|Default|GetStatusWidgetOffsetTimerHandle)))))))

(event Custom|StopUpdatingStatusWidget
  ;[1 1 1] Health Bar
  (bind _statuswidgetcomponent (Variables|FortPawnComponent_PlayAsVillain|GetStatusWidgetComponent)
    (:then
      (bind _asnpcstatus_widget_aiboss (Utilities|Casting|CastToNPCStatusWidget_AIBoss (UserInterface|GetUserWidgetObject _statuswidgetcomponent)))
      (Default|UnbindEventfromVisibilityChanged _asnpcstatus_widget_aiboss (delegate Custom|WidgetVisibilityChanged))
      (Utilities|Time|ClearandInvalidateTimerbyHandle (Variables|Default|GetStatusWidgetOffsetTimerHandle))
      (bind _asfort_pawn (Utilities|Casting|CastToFortPawn (Components|GetOwner)))
      (EventDispatchers|UnbindEventfromOnDBNOStateChanged _asfort_pawn (EventDispatchers|CreateEvent :!event HandlePawnDBNOStateChange)))
    (:else)))

(event Custom|HandleAudioFootstepOverride
  (params (Max%20Footstep%20Distance :type Float%20%28single-precision%29))
  (bind _asfort_pawn (Utilities|Casting|CastToFortPawn (Components|GetOwner)))
  (Utilities|IsValid (Footsteps|GetFootstepComponent _asfort_pawn)
    (:Is%20Valid
      (Variables|/Script/FortniteGame.FortFootstepComponent|SetMaxFootstepDistance Max%20Footstep%20Distance (Footsteps|GetFootstepComponent _asfort_pawn)))))

(event Custom|ListenForCloakedActivation
  (bind (_boutiswithinlineofsight _returnvalue) (CallFunction|CanShowStatusWidget))
  ;[0 0 0 0.5] FORT-1028730: Handle Cloaking Items and Abilities
  (bind _asynctaskproxy (Ability|Tasks|WaitGameplayTagAddToActor (Components|GetOwner) (Variables|Cloaking|GetCloakedTag))
    (:then
      (Variables|Cloaking|SetCloakActiveAction _asynctaskproxy))
    (:Added
      (bind _statuswidgetcomponent (Variables|FortPawnComponent_PlayAsVillain|GetStatusWidgetComponent)
        (:then
          (bind _asnpcstatus_widget_aiboss (Utilities|Casting|CastToNPCStatusWidget_AIBoss (UserInterface|GetWidget _statuswidgetcomponent)))
          (Class|NPCStatusWidgetAIBoss|EventHideStatus _asnpcstatus_widget_aiboss)
          (bind _asynctaskproxy_2 (Ability|Tasks|WaitGameplayTagRemoveFromActor (Components|GetOwner) (Variables|Cloaking|GetCloakedTag))
            (:then
              (Variables|Cloaking|SetCloakDisabledAction _asynctaskproxy_2))
            (:Removed
              (if _returnvalue
                (bind _statuswidgetcomponent_2 (Variables|FortPawnComponent_PlayAsVillain|GetStatusWidgetComponent)
                  (:then
                    (bind _asnpcstatus_widget_aiboss_2 (Utilities|Casting|CastToNPCStatusWidget_AIBoss (UserInterface|GetWidget _statuswidgetcomponent_2)))
                    (Class|NPCStatusWidgetAIBoss|EventShowStatus _asnpcstatus_widget_aiboss_2))
                  (:else))))))
        (:else)))))

(fn CanUpdateStatusWidget
  (if (not (Team|OnLocalPlayersTeam (Components|GetOwner)))
    (bind (_outplayercontroller _outsquareddistance) (CallFunction|GetClosestLocalPlayerController))
    (Utilities|IsValid _outplayercontroller
      (:Is%20Valid
        (if (<= _outsquareddistance (Variables|FortPawnComponent_PlayAsVillain|GetShowStatusWidgetMaxSquaredRadius))
          (if (Variables|FortPawnComponent_PlayAsVillain|GetbShouldCheckLineOfSight)
            (bind _returnvalue (Fortnite|CheckLineOfSightToActorWithChannel (Pawn|GetPlayerViewPoint _outplayercontroller) (Components|GetOwner) "ECC_Visibility" _outplayercontroller))
            (return _returnvalue)
            (else
              (return true))))))))

(fn HandleBossStatusWidgetEnabledChanged
  (Variables|Default|GetHasBegunPlay
    (:then
      (Networking|CanExecuteCosmeticEvents
        (:True
          (if (Variables|Default|GetbBossStatusWidgetEnabled)
            (CallFunction|SetUpStatusWidget)
            (CallFunction|StartUpdatingStatusWidget)
            (bind _returnvalue (Utilities|Time|SetTimerbyEvent (EventDispatchers|CreateEvent :!event UpdateStatusWidgetVisibility) (ScalableFloat|GetValueasDouble (Variables|Hotfix|GetUpdateStatusWidgetVisibilityFrequency)) true true))
            (Variables|Default|SetShowHealthBarCheckTimer _returnvalue)
            (bind _asfort_pawn (Utilities|Casting|CastToFortPawn (Components|GetOwner)))
            (bind _returnvalue_2 (Components|AddDynamicComponentbyClass _asfort_pawn "/Script/FortniteGame.FortSoundIndicatorComponent"))
            (bind _asfort_sound_indicator_component (Utilities|Casting|CastToFortSoundIndicatorComponent _returnvalue_2))
            (Variables|Default|SetFortSoundIndicator _asfort_sound_indicator_component)
            (bind _statuswidgetcomponent (Variables|FortPawnComponent_PlayAsVillain|GetStatusWidgetComponent)
              (:then
                (bind _aswbp_play_as_villain_status (Utilities|Casting|CastToWBP_PlayAsVillain_Status (UserInterface|GetUserWidgetObject _statuswidgetcomponent)))
                (Class|WBPPlayasVillainStatus|SetOvershieldVisbility _aswbp_play_as_villain_status (Class|FortGameStateAthena|PlaylistHasTag (Fortnite|Athena|GameState|GetGameStateAthena) (GameplayTags|MakeLiteralGameplayTag "(TagName=\"Athena.Playlist.NoBuildBR\")")) true))
              (:else))
            (else
              (CallFunction|TearDownStatusWidget)
              (CallFunction|StopUpdatingStatusWidget)
              (Utilities|Time|ClearandInvalidateTimerbyHandle (Variables|Default|GetShowHealthBarCheckTimer)))))
        (:False
          (bind _asfort_pawn_2 (Utilities|Casting|CastToFortPawn (Components|GetOwner)))
          (LocalVariables|SetFortPawnOwner _asfort_pawn_2)
          (Variables|Default|GetbBossStatusWidgetEnabled
            (:then
              (CallFunction|HandlePawnHealthChanged)
              (Health|BindEventtoOnPawnHealthChanged (LocalVariables|GetFortPawnOwner) (EventDispatchers|CreateEvent :!event HandlePawnHealthChanged))
              (bind _asfort_player_state_athena (Utilities|Casting|CastToFortPlayerStateAthena (Variables|/Script/Engine.Pawn|GetPlayerState (LocalVariables|GetFortPawnOwner))))
              (Player|BindEventtoOnOvershieldStateChanged _asfort_player_state_athena (EventDispatchers|CreateEvent :!event HandleOvershieldChanged))
              (bind _asfort_player_state_athena_2 (Utilities|Casting|CastToFortPlayerStateAthena (Variables|/Script/Engine.Pawn|GetPlayerState (LocalVariables|GetFortPawnOwner))))
              (bind (_overshield _overshieldmax) (HealthSet|GetOvershields _asfort_player_state_athena_2))
              (CallFunction|HandleOvershieldChanged _asfort_player_state_athena_2 _overshield _overshieldmax))
            (:else
              (Health|UnbindEventfromOnPawnHealthChanged (LocalVariables|GetFortPawnOwner) (EventDispatchers|CreateEvent :!event HandlePawnHealthChanged))
              (bind _asfort_player_state_athena_3 (Utilities|Casting|CastToFortPlayerStateAthena (Variables|/Script/Engine.Pawn|GetPlayerState (LocalVariables|GetFortPawnOwner))))
              (Player|UnbindEventfromOnOvershieldStateChanged _asfort_player_state_athena_3 (EventDispatchers|CreateEvent :!event HandleOvershieldChanged)))))))
    (:else)))

(fn HandleFortSoundIndicatorStart
  (params (AV%20Icon :type Texture%20Object%20Reference) (AV%20Indicator%20Type :type EFortSoundIndicatorTypes%20Enum :default "Generic") (AV%20Duration :type Float%20%28double-precision%29) (InSoundBeingPlayed :type Sound%20Base%20Object%20Reference))
  (bind _fortsoundindicator (Variables|Default|GetFortSoundIndicator)
    (:then
      (Variables|/Script/FortniteGame.FortSoundIndicatorComponent|SetDefaultIconOverride (Variables|GetAV%20Icon) _fortsoundindicator)
      (Variables|/Script/FortniteGame.FortSoundIndicatorComponent|SetDefaultSoundIndicatorType (Variables|GetAV%20Indicator%20Type) _fortsoundindicator)
      (SoundIndicator|ShowSoundIndicatorforDuration _fortsoundindicator (Transformation|GetActorLocation (Components|GetOwner)) (Variables|GetInSoundBeingPlayed) 1.0 (Variables|GetAV%20Duration) 0))
    (:else)))

(fn HandleOvershieldChanged
  (params (FortPlayerState :type Fort%20Player%20State%20Athena%20Object%20Reference) (Overshield :type Float%20%28single-precision%29) (MaxOvershield :type Float%20%28single-precision%29))
  (Variables|SetBossOverShieldInfo_Overshield_2_F3B52F6A4931C9622E7A65BA8BBE5FDE (Utilities|Struct|MakeSPlayAsVIllainOvershield (Math|Float|SafeDivide Overshield MaxOvershield) MaxOvershield)))

(fn HandlePawnDBNOStateChange
  (params (FortPawn :type Fort%20Pawn%20Object%20Reference) (bInIsDBNO :type Boolean))
  (bind _statuswidgetcomponent (Variables|FortPawnComponent_PlayAsVillain|GetStatusWidgetComponent)
    (:then
      (bind _aswbp_play_as_villain_status (Utilities|Casting|CastToWBP_PlayAsVillain_Status (UserInterface|GetUserWidgetObject _statuswidgetcomponent)))
      (Class|WBPPlayasVillainStatus|SetColoronDBNO _aswbp_play_as_villain_status (Variables|GetbInIsDBNO)))
    (:else)))

(fn HandlePawnHealthChanged
  (bind _asfort_pawn (Utilities|Casting|CastToFortPawn (Components|GetOwner)))
  (Variables|SetBossHealthInfo_HealthPercent_2_7EDADA384C7908EA98DEBD9958A249DE (Utilities|Struct|MakeStructNPCHealthInfo (Math|Float|SafeDivide (FortPawn|GetHealth _asfort_pawn) (FortPawn|GetMaxHealth _asfort_pawn)) (Math|Float|SafeDivide (FortPawn|GetShield _asfort_pawn) (FortPawn|GetMaxShield _asfort_pawn)) (FortPawn|GetMaxHealth _asfort_pawn) (FortPawn|GetMaxShield _asfort_pawn))))

(fn OnRep_bAbilityUsed
  (Networking|CanExecuteCosmeticEvents
    (:True
      (bind _statuswidgetcomponent (Variables|FortPawnComponent_PlayAsVillain|GetStatusWidgetComponent)
        (:then
          (bind _aswbp_play_as_villain_status (Utilities|Casting|CastToWBP_PlayAsVillain_Status (UserInterface|GetUserWidgetObject _statuswidgetcomponent)))
          (LocalVariables|SetWBPPlayAsVillainStatus _aswbp_play_as_villain_status)
          (if (Variables|Default|GetbAbilityUsed)
            (CallFunction|HandleFortSoundIndicatorStart (Variables|AV|GetAbilityUsedAVIcon) (Variables|AV|GetAbilityUsedIndicatorType) (Variables|AV|GetAbilityUsedAVDuration) (Variables|AV|GetAbilityUsedSoundCue))
            (bind _canupdate (CallFunction|CanUpdateStatusWidget))
            (if _canupdate
              (Class|WBPPlayasVillainStatus|SetOvershieldVisbility (LocalVariables|GetWBPPlayAsVillainStatus) true false))
            (else
              (bind _canupdate_2 (CallFunction|CanUpdateStatusWidget))
              (if _canupdate_2
                (Class|WBPPlayasVillainStatus|SetOvershieldVisbility (LocalVariables|GetWBPPlayAsVillainStatus) (Class|FortGameStateAthena|PlaylistHasTag (Fortnite|Athena|GameState|GetGameStateAthena) (GameplayTags|MakeLiteralGameplayTag "(TagName=\"Athena.Playlist.NoBuildBR\")")) false)))))
        (:else)))))

(fn OnRep_bBossStatusWidgetEnabled
  (CallFunction|HandleBossStatusWidgetEnabledChanged))

(fn OnRep_BossHealthInfo
  (bind _canupdate (CallFunction|CanUpdateStatusWidget))
  (if _canupdate
    (CallFunction|UpdateStatusWidgetData false)))

(fn OnRep_BossOverShieldInfo
  (bind _canupdate (CallFunction|CanUpdateStatusWidget))
  (if _canupdate
    (CallFunction|UpdateStatusWidgetData false)))

(fn OnRep_ScaleOverride
  (CallFunction|UpdateScale))

(fn OnRep_TransformationTag
  (bind _asfort_player_pawn (Utilities|Casting|CastToFortPlayerPawn (Components|GetOwner)))
  (Utilities|IsValid (Variables|/Script/Engine.Pawn|GetPlayerState _asfort_player_pawn)
    (:Is%20Valid
      (LocalVariables|SetPlayerState (Variables|/Script/Engine.Pawn|GetPlayerState _asfort_player_pawn))
      (bind _returnvalue (GameplayEventRouter|GetEventRouter "/Script/FortniteGame.FortPlayspace" (LocalVariables|GetPlayerState)))
      (GameplayEventRouter|BroadcastStatefulEvent _returnvalue (Utilities|Struct|MakeGameplayEventPlayerTransformed (LocalVariables|GetPlayerState) (Variables|Default|GetTransformation%20Tag)) (LocalVariables|GetPlayerState))
      (Networking|CanExecuteCosmeticEvents
        (:True
          (CallFunction|HandleAudioFootstepOverride 4e+03)
          (CallFunction|HandleFortSoundIndicatorStart (Variables|AV|GetTransformationAVIcon) (Variables|AV|GetTransformationIndicatorType) (Variables|AV|GetTransformationAVDuration) (Variables|AV|GetTransformationSoundCue)))
        (:False
          (if (not (ScalableFloat|GetValueasBool (Variables|Hotfix|GetVagueTrackingEnabled_HF)))
            (if (not (Variables|Default|GetIsMapActorAttached))
              (CallFunction|SetupMapIcon))))))))

(fn SetupMapIcon
  (bind _returnvalue (Components|GetOwner))
  (bind _returnvalue_2 (Game|SpawnActorfromClass "/PlayAsVillain/Gameplay/B_PlayAsVillain_MinimapActor.B_PlayAsVillain_MinimapActor_C" (Math|Transform|MakeTransform (Transformation|GetActorLocation _returnvalue)) "Undefined" "MultiplyWithRoot" _returnvalue 0))
  (bind _output_get (Variables|Default|SetMapIconActor _returnvalue_2))
  (Transformation|AttachActorToActor _output_get _returnvalue "None" "SnapToTarget" "KeepWorld" "KeepWorld" true)
  (Variables|Default|SetIsMapActorAttached true))

(fn SetUpStatusWidget
  (bind (_bosshealthinfo_healthpercent_2_7edada384c7908ea98debd9958a249de _bosshealthinfo_shieldpercent_5_b79c63b64aacf275a1621a9315964dcd _bosshealthinfo_maxhealth_6_7d7bef7841b13fb96d2578989cf135bb _bosshealthinfo_maxshield_9_f923e4444e2e476c56fb0e9d9d2f7846) (Utilities|Struct|BreakStructNPCHealthInfo (Variables|GetBossHealthInfo_HealthPercent_2_7EDADA384C7908EA98DEBD9958A249DE)))
  (bind (_bossovershieldinfo_overshield_2_f3b52f6a4931c9622e7a65ba8bbe5fde _bossovershieldinfo_maxovershield_4_74a18c3f422f55d98ae2cebe4f06999e) (Utilities|Struct|BreakSPlayAsVIllainOvershield (Variables|GetBossOverShieldInfo_Overshield_2_F3B52F6A4931C9622E7A65BA8BBE5FDE)))
  (bind _asfort_pawn (Utilities|Casting|CastToFortPawn (Components|GetOwner)))
  (bind _output_get (LocalVariables|SetFortPawnOwner _asfort_pawn))
  (bind _capsulecomponent (Variables|/Script/Engine.Character|GetCapsuleComponent _output_get)
    (:then
      (bind _mesh (Variables|/Script/Engine.Character|GetMesh _output_get)
        (:then
          (bind _returnvalue (Components|AddDynamicComponentbyClass _output_get "/PlayAsVillain/Gameplay/B_PlayAsVillain_StatusWidgetComponent.B_PlayAsVillain_StatusWidgetComponent_C"))
          (bind _asb_play_as_villain_status_widget_component (Utilities|Casting|CastToB_PlayAsVillain_StatusWidgetComponent _returnvalue))
          (bind _output_get_2 (Variables|FortPawnComponent_PlayAsVillain|SetStatusWidgetComponent _asb_play_as_villain_status_widget_component))
          (Transformation|AttachComponentToComponent _output_get_2 _mesh "None" "SnapToTarget" "SnapToTarget" "SnapToTarget" false)
          (Transformation|SetRelativeLocation (Variables|FortPawnComponent_PlayAsVillain|GetStatusWidgetComponent) (Math|Vector|MakeVector 0.0 0.0 (+ (* (Components|Capsule|GetUnscaledCapsuleHalfHeight _capsulecomponent) 2.0) 3e+01)) false true)
          (bind _aswbp_play_as_villain_status (Utilities|Casting|CastToWBP_PlayAsVillain_Status (UserInterface|GetUserWidgetObject (Variables|FortPawnComponent_PlayAsVillain|GetStatusWidgetComponent))))
          (Class|WBPPlayasVillainStatus|SetPlayerName _aswbp_play_as_villain_status (LocalVariables|GetFortPawnOwner))
          (Class|NPCStatusWidgetAIBoss|EventUpdateStatus _aswbp_play_as_villain_status _bosshealthinfo_healthpercent_2_7edada384c7908ea98debd9958a249de _bosshealthinfo_shieldpercent_5_b79c63b64aacf275a1621a9315964dcd _bosshealthinfo_maxhealth_6_7d7bef7841b13fb96d2578989cf135bb _bosshealthinfo_maxshield_9_f923e4444e2e476c56fb0e9d9d2f7846 true false)
          (Class|WBPPlayasVillainStatus|EventOvershieldUpdate _aswbp_play_as_villain_status _bossovershieldinfo_overshield_2_f3b52f6a4931c9622e7a65ba8bbe5fde _bossovershieldinfo_maxovershield_4_74a18c3f422f55d98ae2cebe4f06999e true))
        (:else)))
    (:else)))

(fn TearDownStatusWidget
  (bind _asfort_pawn (Utilities|Casting|CastToFortPawn (Components|GetOwner)))
  (LocalVariables|SetFortPawnOwner _asfort_pawn)
  (bind _statuswidgetcomponent (Variables|FortPawnComponent_PlayAsVillain|GetStatusWidgetComponent)
    (:then
      (Components|DestroyDynamicComponent (LocalVariables|GetFortPawnOwner) _statuswidgetcomponent)
      (Variables|FortPawnComponent_PlayAsVillain|SetStatusWidgetComponent 0))
    (:else)))

(fn UpdateScale
  (bind _returnvalue (Components|GetOwner))
  (bind _scaleoverride (Variables|Default|GetScaleOverride))
  (Utilities|IsValid _returnvalue
    (:Is%20Valid
      (LocalVariables|SetCurrentActorScale (Transformation|GetActorScale3D _returnvalue))
      (if (<= _scaleoverride 0)
        (LocalVariables|SetDesiredActorScale "1.000000,1.000000,1.000000")
        (else
          (LocalVariables|SetDesiredActorScale (Math|Vector|MakeVector _scaleoverride _scaleoverride _scaleoverride))))
      (if (Math|Vector|NotEqualExactly(Vector) (LocalVariables|GetCurrentActorScale) (LocalVariables|GetDesiredActorScale))
        (Transformation|SetActorScale3D _returnvalue (LocalVariables|GetDesiredActorScale))
        ;[1 1 1] Reinitialize physics asset (scales hit volumes)
        (bind _ascharacter (Utilities|Casting|CastToCharacter _returnvalue))
        (Components|SkinnedMesh|SetPhysicsAsset (Variables|/Script/Engine.Character|GetMesh _ascharacter) "/Game/Characters/Player/Male/Male_Avg_Base/Fortnite_M_Avg_Player_Physics.Fortnite_M_Avg_Player_Physics" true)))))

(fn UpdateStatusWidgetData
  (params (InstantUpdate :type Boolean))
  (bind (_bossovershieldinfo_overshield_2_f3b52f6a4931c9622e7a65ba8bbe5fde _bossovershieldinfo_maxovershield_4_74a18c3f422f55d98ae2cebe4f06999e) (Utilities|Struct|BreakSPlayAsVIllainOvershield (Variables|GetBossOverShieldInfo_Overshield_2_F3B52F6A4931C9622E7A65BA8BBE5FDE)))
  (bind (_bosshealthinfo_healthpercent_2_7edada384c7908ea98debd9958a249de _bosshealthinfo_shieldpercent_5_b79c63b64aacf275a1621a9315964dcd _bosshealthinfo_maxhealth_6_7d7bef7841b13fb96d2578989cf135bb _bosshealthinfo_maxshield_9_f923e4444e2e476c56fb0e9d9d2f7846) (Utilities|Struct|BreakStructNPCHealthInfo (Variables|GetBossHealthInfo_HealthPercent_2_7EDADA384C7908EA98DEBD9958A249DE)))
  (bind _statuswidgetcomponent (Variables|FortPawnComponent_PlayAsVillain|GetStatusWidgetComponent)
    (:then
      (bind _aswbp_play_as_villain_status (Utilities|Casting|CastToWBP_PlayAsVillain_Status (UserInterface|GetUserWidgetObject _statuswidgetcomponent)))
      (Class|NPCStatusWidgetAIBoss|EventUpdateStatus _aswbp_play_as_villain_status _bosshealthinfo_healthpercent_2_7edada384c7908ea98debd9958a249de _bosshealthinfo_shieldpercent_5_b79c63b64aacf275a1621a9315964dcd _bosshealthinfo_maxhealth_6_7d7bef7841b13fb96d2578989cf135bb _bosshealthinfo_maxshield_9_f923e4444e2e476c56fb0e9d9d2f7846 true (Variables|GetInstantUpdate))
      (Class|WBPPlayasVillainStatus|EventOvershieldUpdate _aswbp_play_as_villain_status _bossovershieldinfo_overshield_2_f3b52f6a4931c9622e7a65ba8bbe5fde _bossovershieldinfo_maxovershield_4_74a18c3f422f55d98ae2cebe4f06999e (Variables|GetInstantUpdate)))
    (:else)))

(fn UpdateStatusWidgetVisibility
  (bind (_boutiswithinlineofsight _returnvalue) (CallFunction|CanShowStatusWidget))
  (bind _returnvalue_2 (GameplayTags|HasMatchingGameplayTag(Message) (Components|GetOwner) (Variables|Cloaking|GetCloakedTag)))
  (LocalVariables|SetIsCloaked _returnvalue_2)
  (Variables|FortPawnComponent_PlayAsVillain|GetStatusWidgetComponent
    (:then
      (bind _asnpcstatus_widget_aiboss (Utilities|Casting|CastToNPCStatusWidget_AIBoss (UserInterface|GetWidget (Variables|FortPawnComponent_PlayAsVillain|GetStatusWidgetComponent))))
      (if (!= (Widget|IsVisible (UserInterface|GetWidget (Variables|FortPawnComponent_PlayAsVillain|GetStatusWidgetComponent))) _returnvalue)
        (if (and _returnvalue (not (LocalVariables|GetIsCloaked)))
          (Class|NPCStatusWidgetAIBoss|ChangeWidgetVisibleDuration _asnpcstatus_widget_aiboss 0.0)
          (CallFunction|UpdateStatusWidgetData true)
          (Class|NPCStatusWidgetAIBoss|EventShowStatus _asnpcstatus_widget_aiboss)
          (else
            (Class|NPCStatusWidgetAIBoss|ChangeWidgetVisibleDuration _asnpcstatus_widget_aiboss (ScalableFloat|GetValueasDouble (select _boutiswithinlineofsight (Variables|Hotfix|GetDefaultHideDelay) (Variables|Hotfix|GetOutLineOfSightHideDelay))))))))
    (:else)))

(fn UpdateWidgetOffset
  (bind _statuswidgetcomponent (Variables|FortPawnComponent_PlayAsVillain|GetStatusWidgetComponent))
  (bind _returnvalue (UserInterface|GetUserWidgetObject _statuswidgetcomponent))
  (Variables|FortPawnComponent_PlayAsVillain|GetStatusWidgetComponent
    (:then
      (bind _outplayercontroller (CallFunction|GetClosestLocalPlayerController))
      (Utilities|IsValid _outplayercontroller
        (:Is%20Valid
          (bind _returnvalue_2 (AI|GetNPCWidgetOffset _outplayercontroller (Components|GetOwner)))
          (Utilities|IsValid _returnvalue
            (:Is%20Valid
              (Variables|/Script/UMG.Widget|SetRenderTransform_Translation_X (Utilities|Struct|MakeWidgetTransform (Math|Vector2D|MakeVector2D 0.0 _returnvalue_2)) _returnvalue))))))
    (:else)))
