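; Play As Villain minimap actor that manages the special-actor map icon of its owning pawn.
; A tick after BeginPlay, it matches owner tags against a hide-tag map, tags squad PlayerStates and then sets render tags to hide the icon from them.
; Unless a hotfix flag disables it, gameplay-tag wait tasks swap icon brushes as the owner cloaks and uncloaks.
; On owner death, it removes squad tags and destroys itself.
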(event EventBeginPlay
  (bind _asfort_pawn (Utilities|Casting|CastToFortPawn (Actor|GetOwner)))
  (EventDispatchers|BindEventtoOnDied _asfort_pawn (EventDispatchers|CreateEvent :!event Custom|OnDied))
  (Utilities|FlowControl|DelayUntilNextTick)
  (CallFunction|SetMapIconVisibility)
  (if (not (ScalableFloat|GetValueasBool (Variables|Hotfix|GetShowMapIconWhenCloakedHF)))
    (CallFunction|ListenforCloakActivation)))

(event Custom|OnDied
  (params (DamagedActor :type Actor%20Object%20Reference) (Damage :type Float%20%28single-precision%29) (InstigatedBy :type Controller%20Object%20Reference) (DamageCauser :type Actor%20Object%20Reference) (HitLocation :type Vector) (FHitComponent :type Primitive%20Component%20Object%20Reference) (BoneName :type Name) (Momentum :type Vector))
  (CallFunction|RemoveHideGameplayTags)
  (Actor|DestroyActor))

(event Custom|ListenForCloakActivation
  ;[0 0 0 0.5] FORT-1028730: Add ability to hide Map Icon when the Owner is Cloaked
  (Ability|Tasks|WaitGameplayTagAddToActor (Actor|GetOwner) (Variables|Default|GetCloakedTag)
    (:Added
      (CallFunction|ShowHideMapIcon false)
      (Ability|Tasks|WaitGameplayTagRemoveFromActor (Actor|GetOwner) (Variables|Default|GetCloakedTag)
        (:Removed
          (CallFunction|ShowHideMapIcon true))))))

(fn RemoveHideGameplayTags
  (bind _asfort_player_controller_athena (Utilities|Casting|CastToFortPlayerControllerAthena (Fortnite|GetFortPlayerControllerfromActor (Actor|GetOwner))))
  (bind (_keypin _valuepin) (Utilities|Map|ForEachLoop(Map) (Variables|SpecialActorTags|GetHideMiniMapTagMap))
    (:then
      (bind _returnvalue (GameplayTags|HasMatchingGameplayTag(Message) (Actor|GetOwner) _keypin))
      (bind _output_get (LocalVariables|SetHasMatchingTag _returnvalue))
      (if _output_get
        _))
    (:CompletedPin
      _))
  (if (LocalVariables|GetHasMatchingTag)
    (for _array_element (Team|GetSquadMemberPlayerStates _asfort_player_controller_athena true)
      (Ability|GameplayEffect|RemoveLooseGameplayTags _array_element (GameplayTags|MakeGameplayTagContainerfromTag _valuepin) true))
    (return)))

(fn SetMapIconVisibility
  (bind (_keypin _valuepin) (Utilities|Map|ForEachLoop(Map) (Variables|SpecialActorTags|GetHideMiniMapTagMap))
    (:then
      (bind _returnvalue (GameplayTags|HasMatchingGameplayTag(Message) (Actor|GetOwner) _keypin))
      (bind _output_get (LocalVariables|SetHasMatchingGameplayTag _returnvalue))
      (if _output_get
        (GameplayTags|AddGameplayTag (Variables|SpecialActorTags|GetHideTagContainer) _keypin :!class /Script/GameplayTags.BlueprintGameplayTagLibrary)))
    (:CompletedPin
      _))
  (if (LocalVariables|GetHasMatchingGameplayTag)
    (bind _asfort_player_controller_athena (Utilities|Casting|CastToFortPlayerControllerAthena (Fortnite|GetFortPlayerControllerfromActor (Actor|GetOwner))))
    (for _array_element (Team|GetSquadMemberPlayerStates _asfort_player_controller_athena true)
      (Ability|GameplayEffect|AddLooseGameplayTags _array_element (GameplayTags|MakeGameplayTagContainerfromTag _valuepin) true))
    (Class|AthenaSpecialActorComponent|SetRenderTags (Variables|Default|GetAthenaSpecialActor) (LocalVariables|GetEmptyRenderTags) (GameplayTags|MakeGameplayTagContainerfromTag _valuepin)) ; This nees to be called after Tags are assigned to the Actors\nif done before, some players may not hide the map icon
  ))

(fn ShowHideMapIcon
  (params (IsMapIconActive :type Boolean))
  (Class|AthenaSpecialActorComponent|SetIcons (Variables|Default|GetAthenaSpecialActor) (select (Variables|GetIsMapIconActive) (LocalVariables|GetActiveIconBrush) (LocalVariables|GetHiddenIconBrush)) (select (Variables|GetIsMapIconActive) (LocalVariables|GetActiveIconBrush) (LocalVariables|GetHiddenIconBrush)) (LocalVariables|GetHiddenIconBrush)))

(fn ConstructionScript)
