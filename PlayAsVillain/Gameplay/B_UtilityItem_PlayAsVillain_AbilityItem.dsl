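; Weapon item Blueprint for a Play As Villain ability utility item, tying a looping equip gameplay cue to its instigator.
; On the authority, equipping adds the AbilityEquipped looping cue to the instigator, and unequipping removes it.
; CanApplyHolster always returns false.
; BeginPlay and Tick are empty, and the construction script only calls its parent.
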(event EventBeginPlay)

(event EventTick
  (params (DeltaSeconds :type Float%20%28single-precision%29)))

(event FortWeapon|EventOnEquippedAndAttached
  (Networking|SwitchHasAuthority
    (:Authority
      (GameplayCue|AddGameplayCueOnActor(Looping) (Game|GetInstigator) (Variables|Default|GetAbilityEquipped_GCNL) (Ability|GameplayCue|MakeGameplayCueParameters 0.0)))))

(event FortWeapon|EventOnUnEquippedAndDetached
  (Networking|SwitchHasAuthority
    (:Authority
      (GameplayCue|RemoveGameplayCueOnActor(Looping) (Game|GetInstigator) (Variables|Default|GetAbilityEquipped_GCNL) (Ability|GameplayCue|MakeGameplayCueParameters 0.0)))))

(fn CanApplyHolster
  (params (HolsterId :type Name%20%28by%20ref%29 :default "None"))
  (return false))

(fn ConstructionScript
  (|CallParentFunction))
